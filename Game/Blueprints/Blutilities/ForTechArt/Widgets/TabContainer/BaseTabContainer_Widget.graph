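# BaseTabContainer_Widget is a layout-less base tab container for tech-art editor utility widgets, built from InitialTabs and InitialTabStyles maps on Construct.
# Tab data lives in an array; revalidateTabs rebuilds TA_Generic01_List buttons and WidgetSwitcher pages from it, applying non-default styles and restoring selection by name.
# List selection switches pages, fires onTabSelected and messages pages implementing TA_GenericWidget_Interface.
# Tabs can be added, inserted or removed by name; addTab rejects invalid classes and duplicate names.
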WidgetBlueprint /Game/Blueprints/Blutilities/ForTechArt/Widgets/TabContainer/BaseTabContainer_Widget
graphs [
 EventGraph {nodes [
   Event Construct()
   > initGenericParentWidgetVars()
   > initTabs()
  ]}
 addTab {nodes [
   n1 = addTab!() -> (
    then
    tabName:String
    tabWidgetToLoad:User Widget Class Reference
    insertIndex:int=-1
    ValidateVisuals:bool=true
    OptionalTabStyle:TCW Button Visuals Struct Structure
   ) {is=title K2Node_FunctionEntry}
   > n2 = Sequence!(execute) -> (then_0, then_1, then_2:Exec, then_3:Exec, then_4:Exec, then_5:Exec)
   > n3 = Branch(Condition=IsValidClass!(Class:Object Class Reference=Get tabWidgetToLoad() -> (tabWidgetToLoad:User Widget Class Reference)) -> (ReturnValue:bool))
   setErrorText!(
    execute=@n3.else
    self
    InString:String=Append!(A="Failed to add tab called: ", B, C:String=", as no valid widget class was supplied.") -> (ReturnValue)
   ) -> (then)
   > Return Node!(execute, newTab:TCW Tab Data Struct Structure)
   n4 = For Each Loop(
    Exec=@n2.then_1
    Array:Array of TCW Tab Data Struct Structures=Get currentTabData() -> (currentTabData:Array of TCW Tab Data Struct Structures)
   ) -> (Array Element:TCW Tab Data Struct Structure)
   > Branch(Condition="Equal Exactly (String)"(
     A=Break TCW Tab Data Struct!(TCW_TabData_Struct:"TCW Tab Data Struct Structure (by ref)"=@n4) -> (
      tabName_2_#1:String
      tabIndex_5_#2:int
      tabWidgetClass_11_#3:User Widget Class Reference
      tabContentWidget_14_#4:User Widget Object Reference
      tabButtonWidget_17_#5:User Widget Object Reference
      tabButtonStyle_20_#6:TCW Button Visuals Struct Structure
     ) {is=title K2Node_BreakStruct}
     B=Get tabName() -> (tabName:String)
    ))
   > setErrorText(InString=Append(A="A tab named ", B=Get tabName(), C=" already exists, cannnot add it more than once."))
   > Return Node()
   n5 = Last Index!(TargetArray:Array of TCW Tab Data Struct Structures=Get currentTabData()) -> (ReturnValue=0)
   n6 = Get insertIndex() -> (insertIndex:int=0)
   n7 = Branch(
    execute=@n2.then_2
    Condition=OR Boolean(
     A="integer < integer"!(A:int=@n6, B:int=0) -> (ReturnValue:bool) {is=title K2Node_CallFunction}
     B="integer > integer"(A=@n6, B=@n5) {is=title K2Node_CallFunction}
    )
   )
   > n8 = Set!(
    execute
    l_newData_tabName_2_#1:String=Get tabName()
    l_newData_tabIndex_5_#2:int=Length(TargetArray:Array of TCW Tab Data Struct Structures=Get currentTabData())
    l_newData_tabWidgetClass_11_#3:User Widget Class Reference=Get tabWidgetToLoad()
    l_newData_tabContentWidget_14_#4:User Widget Object Reference
    l_newData_tabButtonWidget_17_#5:User Widget Object Reference
    l_newData_tabButtonStyle_20_#6:TCW Button Visuals Struct Structure=Get OptionalTabStyle() -> (OptionalTabStyle:TCW Button Visuals Struct Structure)
   ) -> (then, Output_Get:TCW Tab Data Struct Structure)
   > Add(TargetArray:Array of TCW Tab Data Struct Structures=Get currentTabData(), NewItem:"TCW Tab Data Struct Structure (by ref)"=@n8) {is=title K2Node_CallArrayFunction}
   Set(
    execute=@n7.else
    l_newData_tabName_2_#1=Get tabName()
    l_newData_tabIndex_5_#2=Get insertIndex() -> (insertIndex=0)
    l_newData_tabWidgetClass_11_#3=Get tabWidgetToLoad()
    l_newData_tabButtonStyle_20_#6=Get OptionalTabStyle()
   )
   > Insert!(
    execute
    TargetArray:Array of TCW Tab Data Struct Structures=Get currentTabData()
    NewItem:"TCW Tab Data Struct Structure (by ref)"=Get l_newData() -> (l_newData:TCW Tab Data Struct Structure)
    Index:int=Get insertIndex() -> (insertIndex=0)
   ) -> (then) {is=title K2Node_CallArrayFunction}
   Branch(execute=@n2.then_3, Condition=@n1.ValidateVisuals)
   > revalidateTabs()
   n9 = getTabDataByName!(execute=@n2.then_4, self, name:String=Get tabName()) -> (then, valid:bool, Array Element:TCW Tab Data Struct Structure)
   > Branch(Condition=@n9)
   > Set l_newData(l_newData:TCW Tab Data Struct Structure=@n9.Array Element) -> (Output_Get:TCW Tab Data Struct Structure)
   > Call On Tab Added!(
    execute:Exec
    self:Base Tab Container Widget Object Reference
    TabContainer:Base Tab Container Widget Object Reference=Self-Reference()
    NewTabName:TCW Tab Data Struct Structure=@n9.Array Element
   ) -> (then:Exec) {is=title K2Node_CallDelegate}
   Return Node(execute=@n2.then_5, newTab=Get l_newData())
   Get tabName()
  ]}
 clearTabData {nodes [
   clearTabData() {is=title K2Node_FunctionEntry}
   > Clear(TargetArray:Array of TCW Tab Data Struct Structures=Get currentTabData())
   > clearTabVisuals()
  ]}
 clearTabVisuals {nodes [
   clearTabVisuals() {is=title K2Node_FunctionEntry}
   > n1 = Sequence!(execute) -> (then_0, then_1)
   > n2 = Cast To TA_Generic01_List!(execute, Object=Get tabButtonParent() -> (tabButtonParent:Widget Object Reference)) -> (then, CastFailed, AsTA Generic 01 List:TA Generic 01 List Object Reference) {is=title K2Node_DynamicCast}
   > clearItems(self:TA Generic 01 List Object Reference=@n2)
   n3 = Cast To WidgetSwitcher!(execute=@n1.then_1, Object=Get tabWidgetParent() -> (tabWidgetParent:Widget Object Reference)) -> (then, CastFailed, AsWidget Switcher:Widget Switcher Object Reference) {is=title K2Node_DynamicCast}
   > ClearChildren(self=@n3)
  ]}
 createNewTabButton {nodes [
   n1 = createNewTabButton!() -> (then, tabName:String, hasDisplayText:bool=true) {is=title K2Node_FunctionEntry}
   > n2 = Cast To TA_Generic01_List(Object=Get tabButtonParent())
   > n3 = addItemGeneric!(execute, self:TA Generic 01 List Object Reference=@n2, displayString:String=SelectString(A=@n1, bPickA=@n1.hasDisplayText)) -> (then, widget:User Widget Object Reference, data:Object Reference)
   > Return Node!(execute, button:User Widget Object Reference=@n3)
  ]}
 createNewTabWidget {nodes [
   n1 = createNewTabWidget!() -> (then, tabData:"TCW Tab Data Struct Structure (by ref)") {is=title K2Node_FunctionEntry}
   > n2 = Cast To WidgetSwitcher(Object=Get tabWidgetParent())
   > n3 = Create Widget!(
    execute:Exec
    Class:User Widget Class Reference=Break TCW Tab Data Struct(TCW_TabData_Struct=@n1).tabWidgetClass_11_#3
    OwningPlayer:Player Controller Object Reference
   ) -> (then:Exec, ReturnValue:User Widget Object Reference)
   > AddChild(self=@n2, Content=@n3)
   > Return Node!(execute, widget:User Widget Object Reference=@n3)
  ]}
 getCurrentTabData {nodes [
   getCurrentTabData() {is=title K2Node_FunctionEntry}
   > Return Node!(execute, currentTabData:Array of TCW Tab Data Struct Structures=Get currentTabData())
  ]}
 getSelectedTab {nodes [
   getSelectedTab() {is=title K2Node_FunctionEntry}
   > n1 = Cast To TA_Generic01_List(Object=Get tabButtonParent())
   > n2 = getSelectedIndex!(execute, self:TA Generic 01 List Object Reference=@n1) -> (then, index:int=0)
   > n3 = Sequence()
   > n4 = getSelectedData!(execute, self:TA Generic 01 List Object Reference=@n1) -> (then, selectedData:Array of Object References)
   > Branch(Condition=Is Valid Index!(TargetArray:Array of Object References=@n4, IndexToTest:int) -> (ReturnValue:bool))
   > n5 = Cast To TA_GenericListItemData!(execute, Object="Get (a copy)"(Array:Array of Object References=@n4) -> (Output:Object Reference)) -> (then, CastFailed, AsTA Generic List Item Data:TA Generic List Item Data Object Reference) {is=title K2Node_DynamicCast}
   > Set l_name(l_name:String=Get displayString!(self:TA Generic List Item Data Object Reference=@n5) -> (displayString:String)) -> (Output_Get:String)
   Return Node!(execute=@n3.then_1, index:int=@n2, name:String=Get l_name() -> (l_name:String))
  ]}
 getTabDataByName {nodes [
   n1 = getTabDataByName!() -> (then, name:String) {is=title K2Node_FunctionEntry}
   > n2 = For Each Loop with Break!(Exec:Exec, Array:Array of TCW Tab Data Struct Structures=Get currentTabData(), Break:Exec) -> (LoopBody:Exec, Array Element:TCW Tab Data Struct Structure, Array Index:int, Completed:Exec)
   Return Node!(execute=@n2.Completed, valid:bool=false, Array Element:TCW Tab Data Struct Structure)
   > Branch(Condition="Equal Exactly (String)"(A=Break TCW Tab Data Struct(TCW_TabData_Struct=@n2), B=@n1))
   > Return Node(valid=true, Array Element=@n2)
  ]}
 getTabWidgetByIndex {nodes [
   n1 = getTabWidgetByIndex!() -> (then, index:int) {is=title K2Node_FunctionEntry}
   n2 = Get currentTabData()
   > Branch(Condition=Is Valid Index(TargetArray:Array of TCW Tab Data Struct Structures=@n2, IndexToTest=@n1))
   > Return Node!(
    execute
    widget:Widget Object Reference=Break TCW Tab Data Struct(TCW_TabData_Struct="Get (a copy)"(Array:Array of TCW Tab Data Struct Structures=@n2, Dimension 1=@n1) -> (Output:TCW Tab Data Struct Structure)).tabContentWidget_14_#4
   )
  ]}
 handleGenericTabButtonListChanged {nodes [
   n1 = handleGenericTabButtonListChanged!() -> (then, selectedData:Array of Object References, selectedWidgets:Array of User Widget Object References, selectedIndices:Array of Integers) {is=title K2Node_FunctionEntry}
   > Branch(Condition=Is Valid Index(TargetArray:Array of Integers=@n1.selectedIndices))
   > n2 = Cast To WidgetSwitcher(Object=Get tabWidgetParent())
   n3 = "Get (a copy)"(Array:Array of Integers=@n1.selectedIndices) -> (Output:int)
   > SetActiveWidgetIndex(self=@n2, Index=@n3)
   n4 = GetChildAt!(self:Panel Widget Object Reference=@n2, Index:int=@n3) -> (ReturnValue:Widget Object Reference)
   > Call On Tab Selected!(execute:Exec, self:Base Tab Container Widget Object Reference, tabIndex:int=@n3, tabWidget:Widget Object Reference=@n4) -> (then:Exec) {is=title K2Node_CallDelegate}
   > Branch(Condition=Does Object Implement Interface!(
     TestObject:Object Reference=@n4
     Interface:Interface Interface=/Game/Blueprints/Blutilities/ForTechArt/Widgets/Data/TA_GenericWidget_Interface.TA_GenericWidget_Interface_C
    ) -> (ReturnValue:bool))
   > onTabSelected!(execute:Exec, self:Object Reference=@n4) -> (then:Exec) {is=title K2Node_Message}
  ]}
 initGenericParentWidgetVars {nodes [
   initGenericParentWidgetVars() {is=title K2Node_FunctionEntry}
   > n1 = Sequence()
   > n2 = Cast To TA_Generic01_List(Object=Get tabButtonParent())
   > Bind Event to On Item Selection Changed(self:TA Generic 01 List Object Reference=@n2, Delegate=Create Event()) {is=title K2Node_AddDelegate}
   n3 = Cast To CommonWidgetSwitcherLegacy!(execute=@n1.then_1, Object=Get tabWidgetParent()) -> (then, CastFailed, AsCommon Widget Switcher Legacy:Common Widget Switcher Legacy Object Reference) {is=title K2Node_DynamicCast}
   > SetDisableTransitionAnimation!(
    execute
    self:Common Animated Switcher Object Reference=@n3
    bDisableAnimation:bool=Get DisableTabSwitchAnimation() -> (DisableTabSwitchAnimation:bool)
   ) -> (then)
  ]}
 initTabs {nodes [
   initTabs() {is=title K2Node_FunctionEntry}
   > n1 = Keys!(
    execute
    TargetMap:Map of Strings to User Widget Class References=Get InitialTabs() -> (InitialTabs:Map of Strings to User Widget Class References)
   ) -> (then, Keys:Array of Strings)
   > n2 = For Each Loop(Array:Array of Strings=@n1) -> (Array Element:String)
   revalidateTabs(execute=@n2.Completed)
   n3 = Get currentTabData()
   > setTabSelectedByIndex!(
    execute
    self
    indexToSelect:int="Clamp (Integer)"(Value=Get DefaultSelectedIndex() -> (DefaultSelectedIndex:int=0), Min=0, Max=Last Index(TargetArray=@n3) -> (ReturnValue=0)) -> (ReturnValue=0)
   ) -> (then)
   > Call On Tabs Initialized!(execute:Exec, self:Base Tab Container Widget Object Reference, TabContainer:Base Tab Container Widget Object Reference=Self-Reference()) -> (then:Exec) {is=title K2Node_CallDelegate}
   n4 = Values!(execute=@n2.LoopBody, TargetMap:Map of Strings to User Widget Class References=Get InitialTabs()) -> (then, Values:Array of User Widget Class References)
   > addTab!(
    execute
    self
    tabName:String=@n2
    tabWidgetToLoad:User Widget Class Reference="Get (a copy)"(Array:Array of User Widget Class References=@n4, Dimension 1=@n2.Array Index) -> (Output:User Widget Class Reference)
    insertIndex:int=-1
    ValidateVisuals:bool=false
    OptionalTabStyle:TCW Button Visuals Struct Structure=Find(
     TargetMap:Map of Strings to TCW Button Visuals Struct Structures=Get InitialTabStyles() -> (InitialTabStyles:Map of Strings to TCW Button Visuals Struct Structures)
     Key=@n2
    ) -> (Value:TCW Button Visuals Struct Structure)
   ) -> (then, newTab:TCW Tab Data Struct Structure) {is=title K2Node_CallFunction}
  ]}
 isDefaultColor {nodes [
   n1 = Inputs!() -> (" ":Exec, color:Linear Color Structure)
   > n2 = Branch(Condition="Near Equal (LinearColor)"!(A:Linear Color Structure=@n1, B:Linear Color Structure=(R=0.5137,G=0.5137,B=0.5137,A=0.5137), Tolerance:float=0.001) -> (ReturnValue:bool))
   Outputs!(Not Default:Exec=@n2.else, Default:Exec=@n2.then, colorOut:Linear Color Structure=@n1)
  ]}
 onTabAdded {nodes [onTabAdded!() -> (then, TabContainer:Base Tab Container Widget Object Reference, NewTabName:TCW Tab Data Struct Structure) {is=title K2Node_FunctionEntry}]}
 onTabSelected {nodes [onTabSelected!() -> (then, tabIndex:int, tabWidget:Widget Object Reference) {is=title K2Node_FunctionEntry}]}
 onTabsInitialized {nodes [onTabsInitialized!() -> (then, TabContainer:Base Tab Container Widget Object Reference) {is=title K2Node_FunctionEntry}]}
 removeTabByName {nodes [
   n1 = removeTabByName!() -> (then, nameToRemove:String) {is=title K2Node_FunctionEntry}
   > n2 = For Each Loop(Array:Array of TCW Tab Data Struct Structures=Get currentTabData()) -> (Array Element:TCW Tab Data Struct Structure)
   revalidateTabs(execute=@n2.Completed)
   Branch(execute=@n2.LoopBody, Condition="Equal Exactly (String)"(A=@n1, B=Break TCW Tab Data Struct(TCW_TabData_Struct=@n2)))
   > Remove Item!(execute, TargetArray:Array of TCW Tab Data Struct Structures=Get currentTabData(), Item:"TCW Tab Data Struct Structure (by ref)"=@n2) -> (then, ReturnValue:bool)
  ]}
 revalidateTabs {nodes [
   revalidateTabs() {is=title K2Node_FunctionEntry}
   > n1 = getSelectedTab!(execute, self) -> (then, index:int=0, name:String) {is=title K2Node_CallFunction}
   > Set l_lastSelectedName(l_lastSelectedName:String=@n1.name) -> (Output_Get:String)
   > clearTabVisuals() {is=title K2Node_CallFunction}
   > n2 = For Each Loop(Array=Get currentTabData())
   setTabSelectedByIndex(execute=@n2.Completed, indexToSelect=Get l_selectedIndexNew() -> (l_selectedIndexNew:int=0))
   n3 = Break TCW Tab Data Struct(TCW_TabData_Struct=@n2)
   n4 = createNewTabButton!(
    execute=@n2.LoopBody
    self
    tabName:String=@n3
    hasDisplayText:bool=NOT Boolean(A=Break TCW Button Visuals Struct!(TCW_ButtonVisuals_Struct:"TCW Button Visuals Struct Structure (by ref)"=@n3.tabButtonStyle_20_#6) -> (iconOnly_48_#7:bool=~) {is=title K2Node_BreakStruct})
   ) -> (then, button:User Widget Object Reference) {is=title K2Node_CallFunction}
   > validateTabButtonStyle!(
    execute
    self
    button:Widget Object Reference=@n4
    style:TCW Button Visuals Struct Structure=@n3.tabButtonStyle_20_#6
   ) -> (then)
   > n5 = createNewTabWidget!(execute, self, tabData:"TCW Tab Data Struct Structure (by ref)"=@n2) -> (then, widget:User Widget Object Reference) {is=title K2Node_CallFunction}
   > n6 = Set members in TCW Tab Data Struct!(
    execute:Exec
    StructRef:"TCW Tab Data Struct Structure (by ref)"=@n2
    tabContentWidget_14_#4:User Widget Object Reference=@n5
    tabButtonWidget_17_#5:User Widget Object Reference=@n4
   ) -> (then:Exec, StructOut:"TCW Tab Data Struct Structure (by ref)") {is=title K2Node_SetFieldsInStruct}
   > Set Array Elem!(
    execute
    TargetArray:Array of TCW Tab Data Struct Structures=Get currentTabData()
    Index:int=@n2.Array Index
    Item:"TCW Tab Data Struct Structure (by ref)"=@n6
    bSizeToFit:bool=false
   ) -> (then)
   > Branch(Condition="Equal Exactly (String)"(A=@n3, B=Get l_lastSelectedName() -> (l_lastSelectedName:String)))
   > Set l_selectedIndexNew(l_selectedIndexNew:int=@n2.Array Index) -> (Output_Get:int=0)
  ]}
 setCurrentTabData {nodes [
   n1 = setCurrentTabData!() -> (then, currentTabData:Array of TCW Tab Data Struct Structures) {is=title K2Node_FunctionEntry}
   > Set currentTabData(currentTabData:Array of TCW Tab Data Struct Structures=@n1) -> (Output_Get:Array of TCW Tab Data Struct Structures)
   > Return Node!(execute)
  ]}
 setErrorText {nodes [
   n1 = setErrorText!() -> (then, InString:String) {is=title K2Node_FunctionEntry}
   > PrintString!(
    execute
    InString=Append!(A="BaseTabContainer_Widget: ERROR: ", B=@n1) -> (ReturnValue)
    bPrintToScreen
    bPrintToLog
    TextColor
    Duration
    Key
   ) -> (then)
  ]}
 setTabSelectedByIndex {nodes [
   setTabSelectedByIndex!() -> (then, indexToSelect:int) {is=title K2Node_FunctionEntry}
   > n1 = Cast To TA_Generic01_List(Object=Get tabButtonParent())
   > setIndexSelected!(
    execute
    self:TA Generic 01 List Object Reference=@n1
    index:int=Get indexToSelect() -> (indexToSelect:int=0)
    selected:bool=true
    throwEvent:bool=true
   ) -> (then)
  ]}
 validateTabButtonStyle {nodes [
   n1 = validateTabButtonStyle!() -> (then, button:Widget Object Reference, style:TCW Button Visuals Struct Structure) {is=title K2Node_FunctionEntry}
   > Set l_button(l_button:Widget Object Reference=@n1) -> (Output_Get:Widget Object Reference)
   > Set l_style(l_style:TCW Button Visuals Struct Structure=@n1.style) -> (Output_Get:TCW Button Visuals Struct Structure)
   > n2 = Cast To TA_BaseEditorButtonWidget!(execute, Object=Get l_button() -> (l_button:Widget Object Reference)) -> (then, CastFailed, AsTA Base Editor Button Widget:TA Base Editor Button Widget Object Reference) {is=title K2Node_DynamicCast}
   > Set l_asTAButton(l_asTAButton:TA Base Editor Button Widget Object Reference=@n2) -> (Output_Get:TA Base Editor Button Widget Object Reference)
   > n3 = Sequence!(execute) -> (then_0, then_1, then_2:Exec, then_3:Exec, then_4:Exec, then_5:Exec, then_6:Exec, then_7:Exec)
   n4 = Break TCW Button Visuals Struct!(TCW_ButtonVisuals_Struct=Get l_style() -> (l_style:TCW Button Visuals Struct Structure)) -> (
    buttonColor_Default_38_#1:Linear Color Structure
    buttonColor_Selected_40_#8:Linear Color Structure
    labelColor_Default_37_#9:Linear Color Structure
    labelColor_Selected_42_#10:Linear Color Structure
    ButtonStyle_Default_36_#11:Button Style Structure
    ButtonStyle_Selected_35_#12:Button Style Structure
    iconStyle_46_#13:Slate Brush Structure
    iconOnly_48_#7=~
   )
   > n5 = Is Default Color!(" ":Exec, color:Linear Color Structure=@n4) -> (Not Default:Exec, Default:Exec, colorOut:Linear Color Structure) {is=title K2Node_MacroInstance}
   > Set buttonColor_Default!(
    execute
    buttonColor_Default:Linear Color Structure=@n5
    self:TA Base Editor Button Widget Object Reference=Get l_asTAButton() -> (l_asTAButton:TA Base Editor Button Widget Object Reference)
   ) -> (then, Output_Get:Linear Color Structure)
   n6 = Is Default Color(" "=@n3.then_1, color=@n4.buttonColor_Selected_40_#8)
   > Set buttonColor_Selected!(execute, buttonColor_Selected:Linear Color Structure=@n6, self:TA Base Editor Button Widget Object Reference=Get l_asTAButton()) -> (then, Output_Get:Linear Color Structure)
   n7 = Is Default Color(" "=@n3.then_2, color=@n4.labelColor_Default_37_#9)
   > Set labelColor_Default!(execute, labelColor_Default:Linear Color Structure=@n7, self:TA Base Editor Button Widget Object Reference=Get l_asTAButton()) -> (then, Output_Get:Linear Color Structure)
   n8 = Is Default Color(" "=@n3.then_3, color=@n4.labelColor_Selected_42_#10)
   > Set labelColor_Selected!(execute, labelColor_Selected:Linear Color Structure=@n8, self:TA Base Editor Button Widget Object Reference=Get l_asTAButton()) -> (then, Output_Get:Linear Color Structure)
   n9 = Break TCW Button Visuals Struct(TCW_ButtonVisuals_Struct=Get l_style()) -> (iconOnly_48_#7=~)
   Is Valid(
    exec=@n3.then_4
    InputObject=Break Button Style!(ButtonStyle:"Button Style Structure (by ref)"=@n9.ButtonStyle_Default_36_#11) -> (
     Normal_TintColor:Slate Color Structure
     Normal_DrawAs:ESlateBrushDrawType Enum
     Normal_Tiling:ESlateBrushTileType Enum
     Normal_Mirroring:ESlateBrushMirrorType Enum
     Normal_ImageSize:Vector 2D Structure
     Normal_Margin:Margin Structure
     Normal_ResourceObject:Object Reference
     Normal_OutlineSettings:Slate Brush Outline Settings Structure
     Hovered:Slate Brush Structure
     Pressed:Slate Brush Structure
     Disabled:Slate Brush Structure
     NormalForeground:Slate Color Structure
     HoveredForeground:Slate Color Structure
     PressedForeground:Slate Color Structure
     DisabledForeground:Slate Color Structure
     NormalPadding:Margin Structure
     PressedPadding:Margin Structure
     PressedSlateSound:Slate Sound Structure
     ClickedSlateSound:Slate Sound Structure
     HoveredSlateSound:Slate Sound Structure
    ) {is=title K2Node_BreakStruct}.Normal_ResourceObject
   )
   > Set ButtonStyle_Default!(
    execute
    ButtonStyle_Default:Button Style Structure=@n9.ButtonStyle_Default_36_#11
    self:TA Base Editor Button Widget Object Reference=Get l_asTAButton()
   ) -> (then, Output_Get:Button Style Structure)
   Is Valid(
    exec=@n3.then_5
    InputObject=Break Button Style(ButtonStyle=@n9.ButtonStyle_Selected_35_#12).Normal_ResourceObject
   )
   > Set ButtonStyle_Selected!(
    execute
    ButtonStyle_Selected:Button Style Structure=@n9.ButtonStyle_Selected_35_#12
    self:TA Base Editor Button Widget Object Reference=Get l_asTAButton()
   ) -> (then, Output_Get:Button Style Structure)
   n10 = Break TCW Button Visuals Struct(TCW_ButtonVisuals_Struct=Get l_style()) -> (iconOnly_48_#7=~)
   Is Valid(
    exec=@n3.then_6
    InputObject=Break Slate Brush!(SlateBrush:"Slate Brush Structure (by ref)"=@n10.iconStyle_46_#13) -> (
     TintColor:Slate Color Structure
     DrawAs:ESlateBrushDrawType Enum
     Tiling:ESlateBrushTileType Enum
     Mirroring:ESlateBrushMirrorType Enum
     ImageSize:Vector 2D Structure
     Margin:Margin Structure
     ResourceObject:Object Reference
     OutlineSettings:Slate Brush Outline Settings Structure
    ) {is=title K2Node_BreakStruct}.ResourceObject
   )
   > Set iconStyle!(
    execute
    iconStyle:Slate Brush Structure=@n10.iconStyle_46_#13
    self:TA Base Editor Button Widget Object Reference=Get l_asTAButton()
   ) -> (then, Output_Get:Slate Brush Structure)
   validateVisuals!(execute=@n3.then_7, self:TA Base Editor Button Widget Object Reference=Get l_asTAButton()) -> (then, " ":bool)
  ]}
]
variables [
 Current Tab Data {type="array<TCW_TabData_Struct>", default []}
 currentTabData {type="array<TCW_TabData_Struct>", default []}
 DefaultSelectedIndex {type=integer}
 DisableTabSwitchAnimation {type=boolean, default=true}
 InitialTabs {type="map<Class>", default {}}
 InitialTabStyles {type="map<TCW_ButtonVisuals_Struct>", default {}}
 onTabAdded {}
 onTabSelected {}
 onTabsInitialized {}
 tabButtonParent {type=Widget, default=null}
 tabWidgetParent {type=Widget, default=null}
]
widgets {info {inheritedWidgetCount=0, parentClass=&UserWidget, rootWidgetClass=null, widgetCount=0}, widgets []}
