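# Generic tab container widget for tech-art editor utilities, built on BaseTabContainer_Widget.
# A dark bordered vertical layout stacks a horizontal, single-select list of Generic01 tab items above a fill-sized Common widget switcher with short transitions.
# On Construct it assigns the list and switcher as the base class's tab button and tab content parents before calling the parent Construct, leaving tab logic to the base.
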WidgetBlueprint /Game/Blueprints/Blutilities/ForTechArt/Widgets/TabContainer/Generic01_TabContainer
graphs [EventGraph {nodes [
   Event Construct()
   > Set tabButtonParent(tabButtonParent:Widget Object Reference=Get listTabs() -> (listTabs:TA Generic 01 List Object Reference)) -> (Output_Get:Widget Object Reference)
   > Set tabWidgetParent(tabWidgetParent:Widget Object Reference=Get widgetSwitcherTabWidgets() -> (widgetSwitcherTabWidgets:Common Widget Switcher Legacy Object Reference)) -> (Output_Get:Widget Object Reference)
   > Parent: Construct()
   n1 = Event Tick()
   > Parent: Tick!(execute, MyGeometry:Geometry Structure=@n1.MyGeometry, InDeltaTime:float=@n1.InDeltaTime) -> (then) {is=title K2Node_CallParentFunction}
  ]}]
variables []
widgets {
 info {
  inheritedWidgetCount=0
  parentClass=&/BaseTabContainer_Widget.BaseTabContainer_Widget_C
  rootWidgetClass=&Border
  widgetCount=5
 }
 widgets [
  Border_1 {widgetClassPath=&Border, widget=Border_1, properties {brushColor=<rgba 0.0075 0.0075 0.0075 1>, padding=<ltrb 5 5 5 5>}}
  VerticalBox_24 {
   widgetClassPath=&VerticalBox
   parent=Border_1
   widget=VerticalBox_24
   slot {class=BorderSlot, properties {padding=<ltrb 5 5 5 5>}}
  }
  listTabs {
   widgetClassPath=/Game/Blueprints/Blutilities/ForTechArt/Widgets/List/TA_Generic01_List
   bIsVariable=true
   parent=VerticalBox_24
   widget=listTabs
   properties {
    allowMultiSelect=false
    listItemClass=/Game/Blueprints/Blutilities/ForTechArt/Widgets/Buttons/TA_Generic01_TabItem.TA_Generic01_TabItem_C
    scrollOrientation=Orient_Horizontal
   }
   slot {class=VerticalBoxSlot}
  }
  Border_2 {
   widgetClassPath=&Border
   parent=VerticalBox_24
   widget=Border_2
   properties {
    background {
     bIsDynamicallyLoaded=false
     drawAs=Image
     imageType=NoImage
     mirroring=NoMirror
     resourceName=None
     resourceObject=null
     tiling=NoTile
     imageSize=<32 32>
     margin=<ltrb 0 0 0 0>
     outlineSettings {
      bUseBrushTransparency=false
      roundingType=HalfHeightRadius
      width=0
      color {colorUseRule=UseColor_Specified, specifiedColor=<rgba 0 0 0 0>}
      cornerRadii=<0 0 0 0>
     }
     tintColor {colorUseRule=UseColor_Specified, specifiedColor=<rgba 0.0176 0.0176 0.0176 1>}
     uVRegion {bIsValid=false, max=<0 0>, min=<0 0>}
    }
    padding=<ltrb 3 3 3 3>
   }
   slot {class=VerticalBoxSlot, properties {size {sizeRule=Fill, value=1}}}
  }
  widgetSwitcherTabWidgets {
   widgetClassPath=&CommonWidgetSwitcherLegacy
   bIsVariable=true
   parent=Border_2
   widget=widgetSwitcherTabWidgets
   properties {transitionDuration=0.15}
   slot {class=BorderSlot, properties {padding=<ltrb 3 3 3 3>}}
  }
 ]
}
